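# Post-process material that magnifies the center of the screen, apparently for inspecting aliasing, as its name suggests.
# A second scene texture sample, with UVs scaled about the center by a constant, is blended over the normal image through a hard SphereMask sized by the radius parameter.
# The mask's resolution-based UV scale divides R by itself, so it applies no aspect-ratio correction.
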Material /Game/ArtTools/Visualizers/M_AliasVisualization
graph {
 properties {materialDomain=MD_PostProcess, pixelDepthOffsetMode=PDOM_Legacy, usedShadingModels=Default Lit}
 nodes [
  n1 = ScreenPosition()
  n2 = Constant() {r=4}
  n3 = MaterialFunctionCall(Float2=MaterialFunctionCall() {materialFunction=&ScreenResolution}.Visible Resolution) {materialFunction=&BreakOutFloat2Components}
  n4 = AppendVector(A=Constant() {r=1}, B=Divide(A=@n3.R, B=@n3.R))
  LinearInterpolate(
   A=SceneTexture() {sceneTextureId=PPI_PostProcessInput0}.Color
   B=SceneTexture(UVs=MaterialFunctionCall(UVs=@n1.ViewportUV, Texture Scale=@n2) {materialFunction=&ScaleUVsByCenter}.UVs) {sceneTextureId=PPI_PostProcessInput0}.Color
   Alpha=SphereMask(
    A=Constant() {r=0.5}
    B=MaterialFunctionCall(UVs=@n1.ViewportUV, Texture Scale=@n4) {materialFunction=&ScaleUVsByCenter}.UVs
    Radius=ScalarParameter() {defaultValue=0.25, parameterName=radius}
    Hardness=Constant() {r=1}
   )
  )
 ]
}
