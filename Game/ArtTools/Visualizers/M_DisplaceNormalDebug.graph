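# Post-process debug visualizer that compares normals derived from world-position derivatives with the GBuffer world normal.
# A screen-U threshold switches between the two normals, which are shaded by a dot product with an upward-biased camera vector.
# A `radius`-driven SphereMask overlays the result on the scene, but its intended aspect correction divides horizontal resolution by itself, so it has no effect.
# An absolute-difference branch between both shadings is left unconnected.
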Material /Game/ArtTools/Visualizers/M_DisplaceNormalDebug
graph {
 properties {materialDomain=MD_PostProcess, pixelDepthOffsetMode=PDOM_Legacy, usedShadingModels=Default Lit}
 nodes [
  n1 = WorldPosition()
  n2 = Normalize(VectorInput=CrossProduct(A=DDX(Value=@n1.XYZ), B=DDY(Value=@n1.XYZ)))
  n3 = Normalize(VectorInput=Add(A=CameraVectorWS(), B=Constant3Vector() {constant=<rgba 0 0 0.5 0>}))
  n4 = DotProduct(A=@n2, B=@n3)
  n5 = ComponentMask(None=SceneTexture() {sceneTextureId=PPI_WorldNormal}.Color) {b=true, g=true, r=true}
  n6 = DotProduct(A=@n5, B=@n3)
  Abs(None=Subtract(A=Saturate(None=@n4), B=Saturate(None=@n6)))
  n7 = Constant()
  n8 = Constant() {r=0.5}
  n9 = MaterialFunctionCall(Float2=ScreenPosition().ViewportUV) {materialFunction=&BreakOutFloat2Components}
  n10 = StaticBool() {value=true}
  n11 = LinearInterpolate(
   A=@n2
   B=@n5
   Alpha=Ceil(None=MaterialFunctionCall!(ThresholdValue=@n7, ThresholdRange=@n8, Alpha=@n9.R, TwoSidedThreshold=@n10) {materialFunction=/Engine/Functions/Engine_MaterialFunctions02/Masking/ThresholdWithRange}.Result)
  )
  n12 = ScreenPosition()
  n13 = MaterialFunctionCall(Float2=MaterialFunctionCall() {materialFunction=&ScreenResolution}.Visible Resolution) {materialFunction=&BreakOutFloat2Components}
  n14 = AppendVector(A=Constant() {r=1}, B=Divide(A=@n13.R, B=@n13.R))
  LinearInterpolate(
   A=SceneTexture() {sceneTextureId=PPI_PostProcessInput0}.Color
   B=Saturate(None=DotProduct(A=@n11, B=@n3))
   Alpha=SphereMask(
    A=Constant() {r=0.5}
    B=MaterialFunctionCall(UVs=@n12.ViewportUV, Texture Scale=@n14) {materialFunction=&ScaleUVsByCenter}.UVs
    Radius=ScalarParameter() {defaultValue=0.3, parameterName=radius}
    Hardness=Constant() {r=1}
   )
  )
 ]
}
