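# MF_PostSlopeMask is a material function in the art visualizer tools that turns a vector input, treated as a normal, into a slope mask.
# It dots the input's RGB with world-style up (0,0,1), then takes the absolute value and inverts it, so steep areas read bright and flat areas facing up or down read dark.
# The SlopMaskPoint parameter scales the saturated result, and CheapContrast sharpens it with a fixed contrast before output.
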MaterialFunction /Game/ArtTools/Visualizers/MF_PostSlopeMask
graph {nodes [
  n1 = FunctionInput() {inputType=FunctionInput_Vector4}
  n2 = Abs(None=DotProduct(A=ComponentMask(None=@n1) {b=true, g=true, r=true}, B=Constant3Vector() {constant=<rgba 0 0 1 0>}))
  n3 = Multiply(A=Saturate(None=OneMinus(None=@n2)), B=ScalarParameter() {defaultValue=1.5, parameterName=SlopMaskPoint})
  FunctionOutput(None=MaterialFunctionCall(In=Saturate(None=@n3), Contrast=Constant() {r=4}) {materialFunction=&CheapContrast}.Result)
 ]}
