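# Post-process material that visualizes slopes by drawing colored diagonal screen-space stripes and an outline over the scene.
# MF_PostSlopeMask converts world normals into a slope mask that gates the stripes, built from Frac and SphereMask.
# The outline subtracts four offset, ceiled mask samples' average from the center, with offsets scaled by view size.
# Leftover branches for inline slope math and custom-depth debugging do not feed the final blend.
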Material /Game/ArtTools/Visualizers/PP_SlopeVisualize
graph {
 properties {
  blendableLocation=BL_SceneColorAfterDOF
  materialDomain=MD_PostProcess
  pixelDepthOffsetMode=PDOM_Legacy
  shadingModel=MSM_Unlit
  usedShadingModels=Unlit
 }
 nodes [
  n1 = VectorParameter() {parameterName=LineColor, defaultValue=<rgba 1 0.1309 0.0012 0>}
  n2 = MaterialFunctionCall(Float2=MaterialFunctionCall() {materialFunction=&ScreenAlignedUVs}."X 100%, Y 100%") {materialFunction=&BreakOutFloat2Components}
  n3 = Multiply(A=@n2.G) {constB=0.5}
  n4 = Frac(None=Multiply(A=Add(A=@n2.R, B=@n3), B=ScalarParameter() {defaultValue=80, parameterName=LineNum}) {constB=20})
  n5 = Constant()
  n6 = ScalarParameter() {defaultValue=0.2, parameterName=LineWidth}
  n7 = Constant() {r=0.5}
  n8 = SceneTexture() {sceneTextureId=PPI_WorldNormal}
  n9 = LinearInterpolate(
   B=LinearInterpolate(A=ScalarParameter() {defaultValue=0.1, parameterName=LineMinOpacity}, Alpha=SphereMask(A=@n4, B=@n5, Radius=@n6, Hardness=@n7))
   Alpha=MaterialFunctionCall!(In=@n8.Color) {materialFunction=&/MF_PostSlopeMask}.Result
  )
  n10 = VectorParameter() {parameterName=OutlineColor, defaultValue=<rgba 0.7083 0.0325 0 0>}
  n11 = SceneTexture() {sceneTextureId=PPI_WorldNormal}
  n12 = ComponentMask(None=ScreenPosition().ViewportUV) {g=true, r=true}
  n13 = Subtract(A=Divide(A=ViewSize(), B=Constant() {r=320}), B=Constant() {r=2}) {constB=0}
  n14 = Multiply(A=Clamp(None=Ceil(None=@n13)) {maxDefault=2, minDefault=1}, B=SceneTexelSize()) {constA=2}
  n15 = Constant()
  n16 = AppendVector(A=ComponentMask(None=@n14) {desc=x, r=true}, B=@n15)
  n17 = SceneTexture(UVs=Add(A=@n12, B=@n16) {desc=x+}) {sceneTextureId=PPI_WorldNormal}
  n18 = AppendVector(A=@n15, B=ComponentMask(None=@n14) {desc=y, g=true}) {desc=y+}
  n19 = SceneTexture(UVs=Add(A=@n12, B=@n18) {desc=y+}) {sceneTextureId=PPI_WorldNormal}
  n20 = Add(
   A=Ceil(None=MaterialFunctionCall(In=@n17.Color) {materialFunction=&MF_PostSlopeMask}.Result)
   B=Ceil(None=MaterialFunctionCall(In=@n19.Color) {materialFunction=&MF_PostSlopeMask}.Result)
  )
  n21 = Constant() {r=-2}
  n22 = SceneTexture(UVs=Add(A=@n12, B=Multiply(A=@n16, B=@n21)) {desc=x-}) {sceneTextureId=PPI_WorldNormal}
  n23 = SceneTexture(UVs=Add(A=@n12, B=Multiply(A=@n21, B=@n18)) {desc=y-}) {sceneTextureId=PPI_WorldNormal}
  n24 = Add(
   A=Ceil(None=MaterialFunctionCall(In=@n22.Color) {materialFunction=&MF_PostSlopeMask}.Result)
   B=Ceil(None=MaterialFunctionCall(In=@n23.Color) {materialFunction=&MF_PostSlopeMask}.Result)
  )
  n25 = Subtract(
   A=Ceil(None=MaterialFunctionCall(In=@n11.Color) {materialFunction=&MF_PostSlopeMask}.Result)
   B=Divide(A=Add(A=@n20, B=@n24)) {constB=4}
  )
  LinearInterpolate(
   A=LinearInterpolate(A=SceneTexture() {sceneTextureId=PPI_PostProcessInput0}.Color, B=AppendVector(A=@n1.RGB, B=@n1.RGB), Alpha=@n9)
   B=AppendVector(A=@n10.RGB, B=@n10.RGB)
   Alpha=Multiply(A=Saturate(None=@n25), B=ScalarParameter() {parameterName=OutlineOpacity})
  )
  n26 = DotProduct(A=ComponentMask(None=@n8.Color) {b=true, g=true, r=true}, B=Constant3Vector() {constant=<rgba 0 0 1 0>})
  n27 = Saturate(None=OneMinus(None=Abs(None=@n26)))
  n28 = ScalarParameter() {defaultValue=1.5, parameterName=SlopeMaskPoint}
  MaterialFunctionCall(In=Saturate(None=Multiply(A=@n27, B=@n28)), Contrast=Constant() {r=0.5}) {materialFunction=&CheapContrast}
  n29 = SceneTexture() {sceneTextureId=PPI_CustomDepth}
  MaterialFunctionCall!(
   Vector2=@n29.Color
   MaximumNumberOfDigits
   UVs=MaterialFunctionCall() {materialFunction=&ScreenAlignedUVs}."X 100%, Y 100%"
   DebugTextLocation RG_UpperRight BA_LowerLeft
   Component Spacing
  ) {materialFunction=/Engine/Functions/Engine_MaterialFunctions02/Utility/DebugFloat2Values}
  n30 = ComponentMask(None=ScreenPosition().ViewportUV) {g=true, r=true}
  n31 = Subtract(A=Divide(A=ViewSize(), B=Constant() {r=320}), B=Constant() {r=2}) {constB=0}
  n32 = Multiply(A=Clamp(None=Ceil(None=@n31)) {maxDefault=2, minDefault=1}, B=SceneTexelSize()) {constA=2}
  n33 = Constant()
  n34 = AppendVector(A=ComponentMask(None=@n32) {desc=x, r=true}, B=@n33)
  n35 = AppendVector(A=@n33, B=ComponentMask(None=@n32) {desc=y, g=true}) {desc=y+}
  n36 = Constant() {r=-1}
  n37 = Multiply(A=@n34, B=@n36)
  n38 = Multiply(A=@n36, B=@n35)
  MaterialFunctionCall!(
   Base=@n29.Color
   BaseSampleBias
   Scale
   Sample1=SceneTexture(UVs=Add(A=@n30, B=@n34) {desc=x+}) {sceneTextureId=PPI_CustomDepth}.Color
   Sample2=SceneTexture(UVs=Add(A=@n30, B=@n35) {desc=y+}) {sceneTextureId=PPI_CustomDepth}.Color
   Sample3=SceneTexture(UVs=Add(A=@n30, B=@n37) {desc=x-}) {sceneTextureId=PPI_CustomDepth}.Color
   Sample4=SceneTexture(UVs=Add(A=@n30, B=@n38) {desc=y-}) {sceneTextureId=PPI_CustomDepth}.Color
   Divisor=Constant() {r=-16}
   UseCenteredLines
   UseRGB
  ) {materialFunction=/Game/Packages/jl_tests/combine4Samples}
  Saturate(None=Add(A=@n9))
 ]
}
